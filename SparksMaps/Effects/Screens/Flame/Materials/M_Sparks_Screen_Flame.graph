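# Unlit material for Sparks stage screens, rendering animated flames with smoke and rising embers.
# Two panning noise samples, UV-distorted by a third, are gradient-masked and smoothstepped into flames, blended over smoke, with sparkle embers added.
# MF_MojoStageColors supplies primary and secondary colors, while audio beat lifetime drives a remapped curve-atlas Reactivity value.
# The composite feeds MF_Sparks_Screen_Details for screen detailing, with theme parameters exposing tiling, distortion and falloff.
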Material /SparksMaps/Effects/Screens/Flame/Materials/M_Sparks_Screen_Flame
graph {
 properties {
  bAutomaticallySetUsageInEditor=false
  bHasPixelAnimation=true
  bUsedWithNanite=true
  pixelDepthOffsetMode=PDOM_Legacy
  shadingModel=MSM_Unlit
  usedShadingModels=Unlit
 }
 nodes [
  n1 = NamedRerouteUsage()
  n2 = TextureCoordinate()
  n3 = VectorParameter() {group=ScreenThemeParams, parameterName=EmbersTile, defaultValue=<rgba 1 1 0 0.5>}
  n4 = AppendVector(A=@n3.R, B=@n3.R)
  n5 = TextureSample(UVs=Panner(Coordinate=Multiply(A=@n2, B=@n4), Speed=AppendVector(A=@n3.B, B=@n3.B))) {texture=/SparksEffects/Effects/Textures/Tile/T_Sparks_GlowSparkle_Tile}
  n6 = SmoothStep(
   Min=Constant() {r=0.7}
   Max=Constant()
   Value=Power(
    Base=MaterialFunctionCall() {materialFunction=&LinearGradient}.VGradient
    Exp=ScalarParameter() {defaultValue=1.5, group=ScreenThemeParams, parameterName=SmokeStart}
   ) {constExponent=1.5}
  ) {constMax=0.1}
  n7 = Multiply(A=Saturate(None=Add(A=@n1) {constB=0.02}), B=Multiply(A=Power(Base=@n5.G) {constExponent=1.5}, B=@n6))
  n8 = TextureCoordinate()
  n9 = VectorParameter() {group=ScreenThemeParams, parameterName=SmokeTile, defaultValue=<rgba 0.5 0.5 0 0.1>}
  n10 = AppendVector(A=@n9.R, B=@n9.R)
  n11 = TextureSample(UVs=Panner(Coordinate=Multiply(A=@n8, B=@n10), Speed=AppendVector(A=@n9.B, B=@n9.B))) {texture=/Game/Effects/Fort_Effects/Textures/Smoke/T_Clouds_02_M}
  n12 = NamedRerouteUsage()
  n13 = Multiply(
   A=Power(Base=MaterialFunctionCall() {materialFunction=&LinearGradient}.VGradient) {constExponent=1}
   B=Multiply(A=Add(A=@n12) {constB=0.3}) {constB=0.2}
  )
  n14 = NamedRerouteUsage()
  n15 = VectorParameter() {group=ScreenThemeParams, parameterName=Flame_A_Tile, defaultValue=<rgba 3 1.5 0 0.5>}
  n16 = Panner(Coordinate=Multiply(A=TextureCoordinate() {desc="2, 1"}, B=AppendVector(A=@n15.R, B=@n15.R)), Speed=AppendVector(A=@n15.B, B=@n15.B)) {speedY=1.5}
  n17 = TextureCoordinate()
  n18 = VectorParameter() {group=ScreenThemeParams, parameterName=DistortTile, defaultValue=<rgba 1.5 1 0.25 0.15>}
  n19 = AppendVector(A=@n18.R, B=@n18.R)
  n20 = TextureSampleParameter2D(UVs=Panner(Coordinate=Multiply(A=@n17, B=@n19), Speed=AppendVector(A=@n18.B, B=@n18.B))) {
   group=ScreenThemeParams
   parameterName=DistortTexture
   samplerType=SAMPLERTYPE_Grayscale
   texture=&LowResBlurredNoise
  }
  n21 = ScalarParameter() {defaultValue=0.2, group=ScreenThemeParams, parameterName=DistortAmount}
  n22 = TextureObjectParameter() {group=ScreenThemeParams, parameterName=FlameTexture, texture=/Game/Effects/Fort_Effects/Textures/Tile/T_Noise_02}
  n23 = VectorParameter() {group=ScreenThemeParams, parameterName=Flame_B_Tile, defaultValue=<rgba 2 1 0.2 0.25>}
  n24 = Panner(Coordinate=Multiply(A=TextureCoordinate() {desc="2, 1"}, B=AppendVector(A=@n23.R, B=@n23.R)), Speed=AppendVector(A=@n23.B, B=@n23.B)) {speedX=0.2, speedY=1}
  n25 = Add(
   A=TextureSample(UVs=LinearInterpolate(A=@n16, B=@n20.G, Alpha=@n21), Tex=@n22) {samplerType=SAMPLERTYPE_LinearColor, texture=/SparksMaps/Effects/StageShow/Fire/Textures/T_Mojo_Fire_Noise}.G
   B=TextureSample(UVs=LinearInterpolate(A=@n24, B=@n20.G, Alpha=@n21), Tex=@n22) {samplerType=SAMPLERTYPE_LinearColor, texture=&T_Mojo_Fire_Noise}.G
  )
  n26 = Power(
   Base=MaterialFunctionCall() {materialFunction=&LinearGradient}.VGradient
   Exp=ScalarParameter() {defaultValue=1.5, group=ScreenThemeParams, parameterName=GradientPower}
  ) {constExponent=1.5}
  n27 = Saturate(None=SmoothStep(
    Min=ScalarParameter() {group=ScreenThemeParams, parameterName=SmoothStepMin}
    Max=ScalarParameter() {defaultValue=0.2, group=ScreenThemeParams, parameterName=SmoothStepMax}
    Value=Multiply(A=@n25, B=@n26)
   ) {constMax=0.1})
  n28 = LinearInterpolate(A=Multiply(A=@n6, B=Multiply(A=@n11.G, B=@n13)), B=Multiply(A=Multiply(A=@n14), B=@n27) {constB=10}, Alpha=@n27)
  MaterialFunctionCall!(Emissive=Multiply(A=NamedRerouteUsage(), B=Add(A=@n7, B=@n28)), Scanlines Tile) {materialFunction=/SparksMaps/Effects/Screens/Materials/MF/MF_Sparks_Screen_Details}
  n29 = MaterialFunctionCall!(CustomUPosition) {materialFunction=/SparksStageShow/Systems/StageManager/FXManager/MaterialFunctions/MF_MojoStageColors}
  NamedRerouteDeclaration(None=@n29.Primary) {name=PrimaryColor, nodeColor=<rgba 0 0 0 1>}
  n30 = CurveAtlasRowParameter(CurveTime=CollectionParameter() {collection=/FMAudioReactiveData/AudioAnalysis/MPC_Sparks_Audio_Parameters, parameterName=BeatLifetime}) {
   atlas=/SparksMaps/Effects/Screens/Curves/CurveAtlas_Audio_Beat_A
   curve=/SparksMaps/Effects/Screens/Curves/Curve_Audio_Beat_Remap_001
   parameterName=ReactivityCurve
  }
  NamedRerouteDeclaration(None=MaterialFunctionCall(Input=@n30, Input Low=Constant(), Input High=Constant() {r=1}, Target Low=Constant() {r=0.75}, Target High=Constant() {r=2}) {materialFunction=&RemapValueRange}.Result) {name=Reactivity, nodeColor=<rgba 0 0 0 1>}
  NamedRerouteDeclaration(None=@n29.Secondary) {name=SecondaryColor, nodeColor=<rgba 0 0 0 1>}
  NamedRerouteUsage()
  NamedRerouteUsage()
  VectorParameter() {group=ScreenThemeParams, parameterName=TestFlameColor, defaultValue=<rgba 1 0.0737 0.025 1>}
 ]
}
